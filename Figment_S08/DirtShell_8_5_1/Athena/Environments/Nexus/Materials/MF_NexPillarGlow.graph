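# Material function for the Nexus pillar glow, adding animated emissive to incoming material attributes.
# Camera-facing virtual planes around a center project two cracked-glass layers that distort radial distance, driving layered color gradients blended by the Animation input.
# A static isWeapon switch selects either Fresnel masking or vertex-color brightness plus a height-based voting progress bar.
# Forward and mobile paths drop the second crack layer and one core glow.
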MaterialFunction /Figment_S08/DirtShell_8_5_1/Athena/Environments/Nexus/Materials/MF_NexPillarGlow
graph {nodes [
  n1 = CameraVectorWS()
  n2 = VectorParameter() {parameterName=Axis, defaultValue=<rgba 0 0 1 1>}
  n3 = FunctionInput(Preview=VectorParameter() {parameterName=SphereCenter, defaultValue=<rgba -3024 -1776 96 1>}.RGB) {bUsePreviewValueAsDefault=true, inputName=Centerlocation, sortPriority=2}
  n4 = ScalarParameter() {defaultValue=2, parameterName=SimpleScaleGlobal}
  n5 = Multiply(A=ScalarParameter() {defaultValue=61.3648, parameterName=OffsetPlane}, B=@n4)
  n6 = Normalize(VectorInput=Subtract(A=@n3, B=CameraPositionWS()))
  n7 = MaterialFunctionCall!(
   Plane Normal Axis=@n1
   Plane X Axis=Normalize(VectorInput=CrossProduct(A=@n1, B=@n2.RGB))
   Plane Center=Add(A=@n3, B=Multiply(A=@n5, B=@n6))
   UV Scale=Multiply(A=@n4, B=ScalarParameter() {defaultValue=82.912, parameterName=SizeProjectionTexture})
  ) {materialFunction=/Game/Athena/Environments/Nexus/Materials/VirtualPlaneCoordinatesAbs}
  ComponentMask(None=@n7.UVs) {g=true}
  Constant()
  n8 = GetMaterialAttributes(None=FunctionInput() {inputType=FunctionInput_MaterialAttributes}) {attributeGetTypes [#1]}
  n9 = FunctionInput(Preview=Clamp(None=ScalarParameter() {parameterName=Animation})) {bUsePreviewValueAsDefault=true, inputName=Animation, inputType=FunctionInput_Scalar, sortPriority=10}
  n10 = Fresnel(ExponentIn=LinearInterpolate(
    A=ScalarParameter() {defaultValue=1, parameterName=ExpWeaponFadeFrensnel}
    B=ScalarParameter() {defaultValue=1.3809, parameterName=ExpWeaponFadeFrensnelx}
    Alpha=@n9
   )) {baseReflectFraction=0}
  n11 = TextureSample(UVs=@n7.UVs) {texture=/Game/Effects/Fort_Effects/Textures/Tile/T_Glass_Cracks_M}
  n12 = CameraVectorWS()
  n13 = Normalize(VectorInput=CrossProduct(A=@n12, B=VectorParameter() {parameterName=Axis, defaultValue=<rgba 0 0 1 1>}.RGB))
  n14 = ScalarParameter() {defaultValue=20, parameterName=OffsetCenter2}
  n15 = Add(A=@n3, B=Multiply(A=@n6, B=Multiply(A=@n4, B=@n14)))
  n16 = Multiply(A=ScalarParameter() {defaultValue=199.0745, parameterName=SizeProjectionTexture2}, B=@n4)
  n17 = Add(
   A=@n11.R
   B=TextureSample(UVs=MaterialFunctionCall(Plane Normal Axis=@n12, Plane X Axis=@n13, Plane Center=@n15, UV Scale=@n16) {materialFunction=&VirtualPlaneCoordinatesAbs}.UVs) {texture=&T_Glass_Cracks_M}.R
  )
  n18 = ShadingPathSwitch(Default=@n17, Deferred=@n17, Forward=@n11.R, Mobile=@n11.R)
  n19 = QualitySwitch(Default=@n18, High=@n18, Medium=@n18)
  n20 = ScalarParameter() {defaultValue=0.7048, parameterName=Distortion}
  n21 = Constant2Vector() {g=0.5, r=0.5}
  n22 = Fresnel(ExponentIn=ScalarParameter() {defaultValue=0.4967, parameterName=DistortionExp}) {baseReflectFraction=0}
  n23 = ScalarParameter() {defaultValue=0.2744, parameterName=DistortionNormal}
  n24 = Multiply(A=Add(A=Multiply(A=@n19, B=@n20), B=Distance(A=@n7.UVs, B=@n21)), B=OneMinus(None=Multiply(A=@n22, B=@n23)))
  n25 = Divide(
   A=@n24
   B=LinearInterpolate(
    A=ScalarParameter() {defaultValue=3.4486, parameterName=DistanceFalloff}
    B=ScalarParameter() {defaultValue=3.4486, parameterName=DistanceFalloffx}
    Alpha=@n9
   )
  )
  n26 = ScalarParameter() {defaultValue=0.3703, parameterName=FresnelEXP}
  n27 = LinearInterpolate(
   A=VectorParameter() {parameterName=EmColor1, defaultValue=<rgba 0.1636 0.4979 0.9948 1>}.RGB
   B=VectorParameter() {parameterName=EmColor2, defaultValue=<rgba 0 0 0 1>}.RGB
   Alpha=Clamp(None=Power(Base=@n25, Exp=@n26))
  )
  n28 = Multiply(
   A=VectorParameter() {parameterName=CoreGLow, defaultValue=<rgba 1.6276 2.7893 4 1>}.RGB
   B=LinearInterpolate(A=ScalarParameter() {defaultValue=0.4667, parameterName=Glow}, B=ScalarParameter() {defaultValue=1, parameterName=Glowx}, Alpha=@n9)
  )
  n29 = Divide(
   A=@n24
   B=LinearInterpolate(
    A=ScalarParameter() {defaultValue=2.1046, parameterName=DistanceFalloff2}
    B=ScalarParameter() {defaultValue=2.9932, parameterName=DistanceFalloff2x}
    Alpha=@n9
   )
  )
  n30 = ScalarParameter() {defaultValue=8, parameterName=FresnelEXP3}
  n31 = ScalarParameter() {defaultValue=6.4577, parameterName=FresnelEXP3x}
  n32 = Clamp(None=Power(Base=OneMinus(None=@n29), Exp=LinearInterpolate(A=@n30, B=@n31, Alpha=@n9)))
  n33 = ScalarParameter() {defaultValue=1.9398, parameterName=DistanceFalloff3}
  n34 = ScalarParameter() {defaultValue=2.9932, parameterName=DistanceFalloff3x}
  n35 = OneMinus(None=Divide(A=@n24, B=LinearInterpolate(A=@n33, B=@n34, Alpha=@n9)))
  n36 = ScalarParameter() {defaultValue=2.8268, parameterName=FresnelEXP4}
  n37 = Multiply(A=VectorParameter() {parameterName=CoreGLow2, defaultValue=<rgba 2 0 1.9292 1>}.RGB, B=Clamp(None=Power(Base=@n35, Exp=@n36)))
  n38 = Add(A=Add(A=@n27, B=Multiply(A=@n28, B=@n32)), B=@n37)
  n39 = Add(A=@n27, B=@n37)
  n40 = ShadingPathSwitch(Default=@n38, Deferred=@n38, Forward=@n39, Mobile=@n39)
  n41 = VectorParameter() {parameterName=ProgressBar, defaultValue=<rgba 4.8702 1.1458 5 1>}
  n42 = ComponentMask(None=@n7.UVs) {r=true}
  n43 = Multiply(A=Multiply(A=@n42, B=OneMinus(None=@n42))) {constB=3}
  n44 = ScalarParameter() {defaultValue=3.1554, parameterName=ProgressbarFalloff}
  n45 = Subtract(A=WorldPosition() {worldPositionShaderOffset=WPT_ExcludeAllShaderOffsets}.XYZ, B=ActorPositionWS())
  n46 = MaterialFunctionCall() {materialFunction=/Engine/Functions/Engine_MaterialFunctions02/WorldPositionOffset/ObjectScale}
  n47 = Subtract(
   A=Divide(A=ComponentMask(None=@n45) {b=true}, B=Multiply(A=@n46.Scale Z) {constB=770})
   B=LinearInterpolate(
    A=ScalarParameter() {defaultValue=-0.0286, parameterName=ProgressBarBias}
    B=ScalarParameter() {defaultValue=1, parameterName=ProgressBarBiasmax}
    Alpha=ScalarParameter() {parameterName=ProgressBar}
   )
  )
  n48 = Saturate(None=Divide(A=@n47, B=ScalarParameter() {defaultValue=0.005, parameterName=VoteingProgressSoftEdgedistance}))
  n49 = Divide(A=@n47, B=ScalarParameter() {defaultValue=0.169, parameterName=VoteingProgressTopLevel})
  n50 = Divide(A=@n47, B=ScalarParameter() {defaultValue=-1.7287, parameterName=VoteingProgressTopLevel2})
  n51 = Multiply(A=OneMinus(None=Saturate(None=@n49)), B=OneMinus(None=Saturate(None=@n50)))
  n52 = ScalarParameter() {defaultValue=30, parameterName=VoteingProgressTopFalloff}
  n53 = Add(
   A=Multiply(A=Power(Base=@n43, Exp=@n44), B=OneMinus(None=@n48))
   B=Multiply(A=Power(Base=@n51, Exp=@n52), B=ScalarParameter() {defaultValue=1.5351, parameterName=VoteingProgressTopBrightness})
  )
  n54 = LinearInterpolate(
   A=Constant() {r=1}
   B=LinearInterpolate(
    A=ScalarParameter() {defaultValue=0.5976, parameterName=BrightnessButter1}
    B=ScalarParameter() {defaultValue=1.5, parameterName=BrightnessButter2}
    Alpha=VertexColor().R
   )
   Alpha=@n9
  )
  n55 = StaticSwitch(
   True=Multiply(A=Clamp(None=@n10), B=@n40)
   False=Add(A=Multiply(A=@n41.RGB, B=@n53), B=Multiply(A=@n54, B=@n40))
   Value=FunctionInput(Preview=StaticBool()) {bUsePreviewValueAsDefault=true, inputName=isWeapon, inputType=FunctionInput_StaticBool, sortPriority=11}
  )
  FunctionOutput(None=SetMaterialAttributes!(MaterialAttributes=@n8.MaterialAttributes, EmissiveColor=Add(A=@n8.EmissiveColor, B=@n55)) {attributeSetTypes [#1]})
  MaterialFunctionCall() {materialFunction=&ObjectScale}
  OneMinus(None=@n47)
 ]}
